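# MF_RandomNoise_Actor is a material function that turns 2D coordinates into a pseudo-random scalar noise value.
# It uses the classic hash: dot the Coords input with a fixed irrational-looking vector, take the sine, multiply by a large constant and keep the fractional part.
# That fraction interpolates from zero to the NoiseScale input, so the scale sets the upper bound of the output.
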MaterialFunction /LinedNotebookStraitGameplay/Assets/Materials/MF_RandomNoise_Actor
graph {nodes [
  n1 = FunctionInput() {inputName=Coords, inputType=FunctionInput_Vector2}
  n2 = AppendVector(A=Constant() {r=12.9898}, B=Constant() {r=78.233})
  n3 = Multiply(A=Sine(None=DotProduct(A=@n1, B=@n2))) {constB=43758.5469}
  FunctionOutput(None=LinearInterpolate(A=Constant(), B=FunctionInput() {inputName=NoiseScale, inputType=FunctionInput_Scalar}, Alpha=Frac(None=@n3)) {constA=0.1, constB=0.9})
 ]}
