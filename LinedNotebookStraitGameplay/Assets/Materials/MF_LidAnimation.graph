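# Material function that animates a lid opening, going by its names, through World Position Offset with matching rotated normals.
# One curve atlas sample supplies offset-rotation, height, self-rotation and damping progress, timed by custom primitive data or a looping preview switch.
# It blends rotations about the object pivot and an offset pivot, the latter adding a Z-axis matrix spin.
# A local-space translation with per-primitive height is added on top.
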MaterialFunction /LinedNotebookStraitGameplay/Assets/Materials/MF_LidAnimation
graph {nodes [
  n1 = GetMaterialAttributes(None=FunctionInput() {inputName=Material, inputType=FunctionInput_MaterialAttributes}) {attributeGetTypes [#1]}
  n2 = MaterialFunctionCall(Rotation Axis=NamedRerouteUsage(), Rotation Angle=NamedRerouteUsage()) {materialFunction=&FixRotateAboutAxisNormals}
  n3 = NamedRerouteUsage()
  n4 = Cosine(None=@n3)
  n5 = Sine(None=@n3)
  n6 = Constant()
  n7 = MaterialFunctionCall(
   VectorToTransform=Normalize(VectorInput=Transform(None=@n2.New Tangent Space Vertex Normal))
   BasisX=MaterialFunctionCall(X=@n4, Y=Multiply(A=@n5) {constB=-1}, Z=@n6) {materialFunction=&MakeFloat3}.Result
   BasisY=MaterialFunctionCall(X=@n5, Y=@n4, Z=@n6) {materialFunction=&MakeFloat3}.Result
   BasisZ=Constant3Vector() {constant=<rgba 0 0 1 0>}
  ) {materialFunction=&Transform3x3Matrix}
  n8 = LinearInterpolate(
   A=MaterialFunctionCall(Rotation Axis=NamedRerouteUsage(), Rotation Angle=NamedRerouteUsage()) {materialFunction=&FixRotateAboutAxisNormals}.New Tangent Space Vertex Normal
   B=Transform(None=Normalize(VectorInput=@n7.Transformed Vector)) {transformSourceType=TRANSFORMSOURCE_World, transformType=TRANSFORM_Tangent}
   Alpha=Ceil(None=NamedRerouteUsage())
  )
  n9 = LinearInterpolate(A=NamedRerouteUsage(), B=NamedRerouteUsage(), Alpha=Ceil(None=NamedRerouteUsage()))
  n10 = NamedRerouteUsage()
  n11 = Constant()
  n12 = MaterialFunctionCall(
   X=@n11
   Y=@n11
   Z=ScalarParameter() {bUseCustomPrimitiveData=true, defaultValue=300, parameterName=AfterRotateTransformHeight, primitiveDataIndex=1}
  ) {materialFunction=&MakeFloat3}
  n13 = Transform(None=Add(
    A=Multiply(A=@n10, B=@n12.Result)
    B=VectorParameter() {group=LidTransform, parameterName=PivotTransformOffset, defaultValue=<rgba 0 0 20 1>}.RGB
   )) {transformSourceType=TRANSFORMSOURCE_Local}
  FunctionOutput(None=SetMaterialAttributes!(
    MaterialAttributes=@n1.MaterialAttributes
    Normal=MaterialFunctionCall(BaseNormal=@n1.Normal, AdditionalNormal=@n8) {materialFunction=&BlendAngleCorrectedNormals}.Result
    WorldPositionOffset=Add(A=@n9, B=@n13)
   ) {attributeSetTypes [#1, #2]})
  n14 = Multiply(A=Time(), B=ScalarParameter() {defaultValue=0.5, group=LidTransform, parameterName=AnimationSpeed}) {constB=0.3}
  n15 = ScalarParameter() {bUseCustomPrimitiveData=true, group=LidTransform, parameterName=FinalProgressTime, sliderMax=1}
  n16 = CurveAtlasRowParameter(CurveTime=StaticSwitchParameter(True=Frac(None=@n14), False=NamedRerouteDeclaration(None=@n15) {name=FinalTime, nodeColor=<rgba 1 0 0.3047 1>}) {defaultValue=true, group=LidTransform, parameterName=PreviewAnimation}) {
   atlas=/LinedNotebookStraitGameplay/Assets/Textures/A_TimeScale_LidOpen
   curve=/LinedNotebookStraitGameplay/Assets/Textures/TimeScaleTest_LidOpen
   group=LidTransform
   parameterName=TimeRemapSpeedVar
  }
  NamedRerouteDeclaration(None=@n16.R) {name=OffsetRotateProgress, nodeColor=<rgba 1 0 0.4219 1>}
  n17 = Transform(None=NamedRerouteDeclaration(None=VectorParameter() {group=LidTransform, parameterName=LocalRotationAxisFinal, defaultValue=<rgba 1 0 0 1>}.RGB) {name=LocalAxis, nodeColor=<rgba 1 0 0.1406 1>}) {transformSourceType=TRANSFORMSOURCE_Local}
  n18 = Multiply(
   A=NamedRerouteUsage()
   B=ScalarParameter() {defaultValue=-2, desc=This paramter should adjust by hand, group=LidTransform, parameterName=FitRadian}
  )
  NamedRerouteDeclaration(None=RotateAboutAxis(
    NormalizedRotationAxis=@n17
    RotationAngle=NamedRerouteDeclaration(None=@n18) {name=SelfRotateAngle, nodeColor=<rgba 1 0.6094 0 1>}
    PivotPoint=MaterialFunctionCall() {materialFunction=&ObjectPivotPoint}.Object Pivot Location
    Position=WorldPosition().XYZ
   )) {name=SelfRotate, nodeColor=<rgba 1 0 0.1406 1>}
  n19 = Multiply(A=NamedRerouteUsage(), B=ScalarParameter() {defaultValue=-0.24, group=LidTransform, parameterName=OffsetRadian}) {constB=-1}
  n20 = VectorParameter() {group=LidTransform, parameterName=XRotatePivot, defaultValue=<rgba 0 50 0 1>}
  n21 = Transform(None=RotateAboutAxis(
    NormalizedRotationAxis=NamedRerouteUsage()
    RotationAngle=NamedRerouteDeclaration(None=@n19) {name=OffsetRotationAngle, nodeColor=<rgba 0.0781 1 0 1>}
    PivotPoint=NamedRerouteDeclaration(None=@n20.RGB) {name=XPivot, nodeColor=<rgba 1 0.0703 0 1>}
    Position=MaterialFunctionCall() {materialFunction=&LocalPosition}.Local Position
   )) {transformSourceType=TRANSFORMSOURCE_Local}
  n22 = WorldPosition()
  n23 = NamedRerouteUsage()
  n24 = Cosine(None=@n23)
  n25 = Sine(None=@n23)
  n26 = Constant()
  n27 = MaterialFunctionCall(
   VectorToTransform=TransformPosition(None=Add(A=@n21, B=@n22.XYZ)) {transformSourceType=TRANSFORMPOSSOURCE_World}
   BasisX=MaterialFunctionCall(X=@n24, Y=Multiply(A=@n25) {constB=-1}, Z=@n26) {materialFunction=&MakeFloat3}.Result
   BasisY=MaterialFunctionCall(X=@n25, Y=@n24, Z=@n26) {materialFunction=&MakeFloat3}.Result
   BasisZ=Constant3Vector() {constant=<rgba 0 0 1 0>}
   BasisPosition=Constant3Vector()
  ) {materialFunction=&Transform3x3Matrix}
  NamedRerouteDeclaration(None=Subtract(A=TransformPosition(None=@n27.Transformed Vector) {transformType=TRANSFORMPOSSOURCE_World}, B=WorldPosition().XYZ)) {name=OffsetRotate, nodeColor=<rgba 0 1 0.4609 1>}
  NamedRerouteDeclaration(None=@n16.G) {name=HeightProgress, nodeColor=<rgba 0 0.9766 1 1>}
  NamedRerouteDeclaration(None=@n16.B) {name=SelfRotateProgress, nodeColor=<rgba 1 0.8203 0 1>}
  NamedRerouteDeclaration(None=@n16.A) {name=DampingProgress, nodeColor=<rgba 0 0.625 1 1>}
  NamedRerouteDeclaration(None=@n17) {name=WorldAxis, nodeColor=<rgba 0.7813 0 1 1>}
 ]}
