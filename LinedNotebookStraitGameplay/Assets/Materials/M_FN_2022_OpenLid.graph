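# Fortnite 2022 layered master material for Nanite props with an animated opening lid, judging by its name and final lid-animation function.
# MF_GigaMix_2023 feeds shared damage, wind, fan-spin and Curie effects, then DirectFX, jungle-biome and RVT height blending.
# A mobile switch keeps only base color and normal, with constant specular and roughness.
# Creative move-tool, decal projection and lid animation finish the chain; a documentation-link parameter is wired through WPO.
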Material /LinedNotebookStraitGameplay/Assets/Materials/M_FN_2022_OpenLid
graph {
 properties {
  bAutomaticallySetUsageInEditor=false
  bUseMaterialAttributes=true
  bUsedWithNanite=true
  ditheredLODTransition=true
  maxWorldPositionOffsetDisplacement=550
  pixelDepthOffsetMode=PDOM_Legacy
  refractionMethod=RM_IndexOfRefraction
  usedShadingModels=Default Lit
 }
 nodes [
  n1 = MaterialFunctionCall() {materialFunction=/Game/Global/Base4Layers_2022/MaterialFunctions/MF_GigaMix_2023}
  MaterialExpressionReroute_70 = Reroute(None=@n1.Result)
  n2 = ScalarParameter() {
   desc=Documentation link
   group="0 (Expand)  Documentation (RMB - Copy display name) and Debug view"
   parameterName=https://confluence.it.epicgames.com/display/FORT/Fortnite+Base+Material+2022
  }
  n3 = SetMaterialAttributes!(MaterialAttributes=@MaterialExpressionReroute_70.MaterialAttributes, World Position Offset=Multiply(A=@n2)) {attributeSetTypes [#1]}
  MaterialExpressionReroute_71 = Reroute(None=@n3)
  MaterialExpressionReroute_69 = Reroute(None=@MaterialExpressionReroute_71.Output)
  n4 = MaterialFunctionCall(Material_In=@MaterialExpressionReroute_69.MaterialAttributes) {materialFunction=&MF_FN_DestructibleDamage}
  n5 = MaterialFunctionCall!(
   In=MaterialFunctionCall!(Material Attributes=MaterialFunctionCall!(In=@n4.Material_Out) {materialFunction=/Game/Athena/WinterArt/Materials/MF_WPO_WindNew}.Result) {materialFunction=/Game/Creative/Sets/MilitaryBase/Props/MF_FanSpin}.Result
   UseSphericalMappingOverride
  ) {materialFunction=/Game/Packages/Fortress_SharedMaterials/Base_Material_Functions/MF_CurieFallbackEffect}
  MaterialExpressionReroute_68 = Reroute(None=@n5.Result)
  Composite!(None=@MaterialExpressionReroute_68.Result) {desc=Transform to World in case we will use RVT after this point, subgraphName=Tangent to World}
  Composite!(MaterialAttributes=@n1.Result) {subgraphName=Documentation Link}
  Composite(MaterialAttributes=@MaterialExpressionReroute_71.Output) {desc=Game Shared WPO and Fire Effect, subgraphName=WPO And Shared Functions}
  MaterialExpressionReroute_57 = Reroute(None=@n8.RVT H)
  n6 = MaterialFunctionCall!(In=@MaterialExpressionReroute_68.Result) {materialFunction=/Game/ArtTools/Tech_Art/DirectFX/Assets/MaterialFunctions/MF_DirectFX}
  MaterialExpressionReroute_56 = Reroute(None=@n6.Result)
  n7 = MaterialFunctionCall!(AO, RVT_Height=@MaterialExpressionReroute_57."RVT_Height (S)", For Buildings=@MaterialExpressionReroute_56."For Moss (MA)", For Moss) {materialFunction=/Game/Environments/Asteria/Rocks/Materials/MaterialFunctions/MF_JungleBiome}
  MaterialExpressionReroute_58 = Reroute(None=@n7.Buildings Jungl)
  n8 = MaterialFunctionCall!(
   In=@MaterialExpressionReroute_58.Result
   "RVT Height Map (Roughness)"=GetMaterialAttributes(None=@n6.Result) {attributeGetTypes [#2]}.Roughness
  ) {materialFunction=/Game/Global/Base4Layers_2022/MaterialFunctions/MF_RVTBlending}
  Composite!(InputPin=@n8.Result) {subgraphName=Mobile Opt}
  Composite!("For Moss (MA)"=@n6.Result, "RVT_Height (S)"=@n8.RVT H) {subgraphName="Seasonal "}
  MaterialExpressionReroute_54 = Reroute(None=@n8.Result)
  n9 = GetMaterialAttributes(None=@MaterialExpressionReroute_54.InputPin) {attributeGetTypes [#3, #4]}
  n10 = SetMaterialAttributes!(
   MaterialAttributes=@n9.MaterialAttributes
   Specular=Constant() {r=0.5}
   Roughness=Constant() {r=0.8}
   Base Color=@n9.MaterialAttributes
   Normal=@n9.MaterialAttributes
  ) {attributeSetTypes [
    #5
    #2
    #3
    #4
   ]}
  n11 = ShadingPathSwitch(Default=@MaterialExpressionReroute_54.InputPin, Mobile=@n10) {desc=Migrated from FeatureLevelSwitch}
  MaterialExpressionReroute_55 = Reroute(None=@n11)
  MaterialFunctionCall!(Material=MaterialFunctionCall!(MA=MaterialFunctionCall(MA In=@MaterialExpressionReroute_55.Output) {materialFunction=&MF_CreativeMoveTool}.MA Out) {materialFunction=&/MF_DirectDecalProjection_2023}.Result) {materialFunction=&/MF_LidAnimation}
  MaterialExpressionReroute_42 = Reroute(None=@MaterialExpressionReroute_68.Result)
  n12 = GetMaterialAttributes(None=@MaterialExpressionReroute_42.Input) {attributeGetTypes [#4]}
  n13 = SetMaterialAttributes!(MaterialAttributes=@n12.MaterialAttributes, Normal=Transform(None=@n12.Normal)) {attributeSetTypes [#4]}
  PinBase!(Output=@n13) {reroutePins [Output {expression=&self:MaterialExpressionReroute_44}]}
  PinBase!() {reroutePins [Input {expression=&self:MaterialExpressionReroute_42}]}
  PinBase() {reroutePins [MaterialAttributes {expression=&self:MaterialExpressionReroute_70}]}
  PinBase!(Output=@n3) {reroutePins [Output {expression=&self:MaterialExpressionReroute_71}]}
  PinBase!() {reroutePins [MaterialAttributes {expression=&self:MaterialExpressionReroute_69}]}
  PinBase!(Result=@n5.Result) {reroutePins [Result {expression=&self:MaterialExpressionReroute_68}]}
  PinBase!() {reroutePins [InputPin {expression=&self:MaterialExpressionReroute_54}]}
  PinBase!(Output=@n11) {reroutePins [Output {expression=&self:MaterialExpressionReroute_55}]}
  PinBase!() {reroutePins [
    "For Moss (MA)" {expression=&self:MaterialExpressionReroute_56}
    "RVT_Height (S)" {expression=&self:MaterialExpressionReroute_57}
   ]}
  PinBase!(Result=@n7.Buildings Jungl) {reroutePins [Result {expression=&self:MaterialExpressionReroute_58}]}
  MaterialExpressionReroute_44 = Reroute(None=@n13)
 ]
}
