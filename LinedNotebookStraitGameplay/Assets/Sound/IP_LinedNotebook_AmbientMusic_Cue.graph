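# Sound cue for spatialized looping ambient music in the LinedNotebookStrait gameplay content, using sewer music loops and routed to the Fort_Ambient_Athena sound class.
# A random node picks among three equally weighted looping wave players, so each playing instance holds one loop.
# The root attenuation node overrides falloff, adding distance low-pass filtering and listener-focus ducking, with reverb send disabled.
# A concurrency override caps simultaneous instances at a small count.
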SoundCue /LinedNotebookStraitGameplay/Assets/Sound/IP_LinedNotebook_AmbientMusic_Cue
root=Attenuation_1
nodes [
 {
  id=Attenuation_1
  children [Random_1]
  properties {
   bOverrideAttenuation=true
   attenuationOverrides {
    bAttenuateWithLPF=true
    bEnableFocusInterpolation=true
    bEnableListenerFocus=true
    bEnableReverbSend=false
    falloffDistance=1000
    focusReleaseInterpSpeed=0.5
    lPFFrequencyAtMax=2000
    lPFRadiusMax=3000
    lPFRadiusMin=500
    nonFocusVolumeAttenuation=0.5
    occlusionInterpolationTime=1
    occlusionLowPassFilterFrequency=2000
    occlusionTraceChannel=ECC_GameTraceChannel11
    occlusionVolumeAttenuation=0.5
    stereoSpread=512
    attenuationShapeExtents=<250 250>
   }
  }
 }
 {id=Random_1, children [WavePlayer_2,WavePlayer_3,WavePlayer_5], properties {weights [1,1,1]}}
 {
  id=WavePlayer_2
  properties {bLooping=true, soundWaveAssetPtr=&/LinedNotebook_Sewer_AmbientMusic_Loop_01}
 }
 {
  id=WavePlayer_3
  properties {bLooping=true, soundWaveAssetPtr=&/LinedNotebook_Sewer_AmbientMusic_Loop_02}
 }
 {
  id=WavePlayer_5
  properties {bLooping=true, soundWaveAssetPtr=&/LinedNotebook_Sewer_AmbientMusic_Loop_03}
 }
]
properties {
 bOverrideConcurrency=true
 maxDistance=1250
 soundClassObject=&Fort_Ambient_Athena
 volumeMultiplier=0.752
 concurrencyOverrides {maxCount=3}
}
