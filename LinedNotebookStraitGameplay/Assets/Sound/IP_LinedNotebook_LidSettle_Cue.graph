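# Sound cue for the lined notebook lid settle sound, routed through the traps sound class and traps attenuation.
# A Random node picks one of four LidSettle wave variations with equal weight and feeds a Modulator at the root for variation.
# The cue doubles its volume and primes on load.
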SoundCue /LinedNotebookStraitGameplay/Assets/Sound/IP_LinedNotebook_LidSettle_Cue
root=Modulator_1
nodes [
 {id=Modulator_1, children [Random_0]}
 {id=Random_0, children [WavePlayer_3,WavePlayer_4,WavePlayer_0,WavePlayer_1], properties {weights [1,1,1,1]}}
 {id=WavePlayer_0, properties {soundWaveAssetPtr=&/LinedNotebook_LidSettle_03}}
 {id=WavePlayer_1, properties {soundWaveAssetPtr=&/LinedNotebook_LidSettle_04}}
 {id=WavePlayer_3, properties {soundWaveAssetPtr=&/LinedNotebook_LidSettle_01}}
 {id=WavePlayer_4, properties {soundWaveAssetPtr=&/LinedNotebook_LidSettle_02}}
]
properties {
 attenuationSettings=&Fort_Traps_Att
 bPrimeOnLoad=true
 maxDistance=6000
 soundClassObject=&Fort_Traps
 volumeMultiplier=2
}
