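# First-person sound cue for LinedNotebook travel, layering a looping secret-passage travel sound with sewer travel music.
# A Mixer blends the travel loop, held back by a short Delay, with one of three equally weighted, randomly picked music loops, weighting the music much louder.
# Envelopers fade each layer in, the music over a longer curve; the Mixer's first input is empty.
# Playback uses trap attenuation and sound class, and primes on load.
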SoundCue /LinedNotebookStraitGameplay/Assets/Sound/IP_LinedNotebook_Travel_1P_Cue
root=Mixer_0
nodes [
 {id=Delay_0, children [WavePlayer_2], properties {delayMax=0.5, delayMin=0.5}}
 {
  id=Enveloper_0
  children [Delay_0]
  properties {pitchCurve {editorCurveData {keys [{time=0}]}}, volumeCurve {editorCurveData {keys [<at 0 0>,<at 0.25 0.0007>,<at 1 1>]}}}
 }
 {
  id=Enveloper_1
  children [Random_2]
  properties {
   pitchCurve {editorCurveData {keys [{time=0}]}}
   volumeCurve {editorCurveData {keys [
      {interpMode=RCIM_Cubic, tangentMode=RCTM_User, time=0, value=0}
      {arriveTangent=0.8138, interpMode=RCIM_Cubic, tangentMode=RCTM_Break, time=2, value=1}
     ]}}
  }
 }
 {id=Mixer_0, children [null,Enveloper_0,Enveloper_1], properties {inputVolume [1.15,0.6,3]}}
 {id=Random_2, children [WavePlayer_3,WavePlayer_8,WavePlayer_9], properties {weights [1,1,1]}}
 {
  id=WavePlayer_2
  properties {bLooping=true, soundWaveAssetPtr=/FNGameplayCues/Sounds/Fort_Hideables/SecretPassage/Secret_Passage_Travel_V2_Loop}
 }
 {
  id=WavePlayer_3
  properties {bLooping=true, soundWaveAssetPtr=&/LinedNotebook_Sewer_TravelMusic_Loop_01}
 }
 {
  id=WavePlayer_8
  properties {bLooping=true, soundWaveAssetPtr=&/LinedNotebook_Sewer_TravelMusic_Loop_02}
 }
 {
  id=WavePlayer_9
  properties {bLooping=true, soundWaveAssetPtr=&/LinedNotebook_Sewer_TravelMusic_Loop_03}
 }
]
properties {
 attenuationSettings=&Fort_Traps_Att
 bPrimeOnLoad=true
 maxDistance=6000
 soundClassObject=&Fort_Traps
 volumeMultiplier=1.5
}
