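# Sound cue for third-person travel audio in LinedNotebookStraitGameplay, built from secret passage sounds and using traps attenuation and sound class.
# A mixer layers one of three equally weighted random pass-by waves at half volume with a full-volume exit tell wave.
# A modulator fixes the pass-by layer at double pitch, leaving the exit tell unmodulated.
# The cue primes on load and boosts overall volume.
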SoundCue /LinedNotebookStraitGameplay/Assets/Sound/IP_LinedNotebook_Travel_3P_Cue
root=Mixer_0
nodes [
 {id=Mixer_0, children [Modulator_0,WavePlayer_4], properties {inputVolume [0.5,1]}}
 {id=Modulator_0, children [Random_1], properties {pitchMax=2, pitchMin=2}}
 {id=Random_1, children [WavePlayer_1,WavePlayer_2,WavePlayer_3], properties {weights [1,1,1]}}
 {id=WavePlayer_1, properties {soundWaveAssetPtr=/Game/Sounds/Fort_Hideables/SecretPassage/3P_SecretPassage_Passyby_01}}
 {id=WavePlayer_2, properties {soundWaveAssetPtr=/Game/Sounds/Fort_Hideables/SecretPassage/3P_SecretPassage_Passyby_02}}
 {id=WavePlayer_3, properties {soundWaveAssetPtr=/Game/Sounds/Fort_Hideables/SecretPassage/3P_SecretPassage_Passyby_03}}
 {id=WavePlayer_4, properties {soundWaveAssetPtr=/Game/Sounds/Fort_Hideables/SecretPassage/Secret_Passage_Exit_Tell_01}}
]
properties {
 attenuationSettings=&Fort_Traps_Att
 bPrimeOnLoad=true
 maxDistance=6000
 soundClassObject=&Fort_Traps
 volumeMultiplier=2.4992
}
