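# Sound cue that plays a water splash for the LinedNotebook Strait gameplay content.
# A random node picks among four equally weighted sewer splash waves and feeds a modulator at the root, which varies each play.
# It uses the Fort_Traps sound class and attenuation, with a slight volume boost and priming on load.
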SoundCue /LinedNotebookStraitGameplay/Assets/Sound/IP_LinedNotebook_WaterSplash_Cue
root=Modulator_1
nodes [
 {id=Modulator_1, children [Random_0]}
 {id=Random_0, children [WavePlayer_2,WavePlayer_5,WavePlayer_6,WavePlayer_7], properties {weights [1,1,1,1]}}
 {id=WavePlayer_2, properties {soundWaveAssetPtr=&/LinedNotebook_SewerSplash_01}}
 {id=WavePlayer_5, properties {soundWaveAssetPtr=&/LinedNotebook_SewerSplash_02}}
 {id=WavePlayer_6, properties {soundWaveAssetPtr=&/LinedNotebook_SewerSplash_03}}
 {id=WavePlayer_7, properties {soundWaveAssetPtr=&/LinedNotebook_SewerSplash_04}}
]
properties {
 attenuationSettings=&Fort_Traps_Att
 bPrimeOnLoad=true
 maxDistance=6000
 soundClassObject=&Fort_Traps
 volumeMultiplier=1.3
}
