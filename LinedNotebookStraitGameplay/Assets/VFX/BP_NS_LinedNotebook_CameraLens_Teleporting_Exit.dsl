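; Niagara actor Blueprint for a camera lens effect, named for the teleport exit in the LinedNotebook strait gameplay.
; On BeginPlay it reads the viewport size and compares its aspect ratio to a 16:9 baseline.
; The scaled ratio, clamped to a minimum of one, sets the Niagara "ViewportScale" float so the effect grows on sufficiently wide screens.
; The scale is computed only once, as Tick is disabled.
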(event EventBeginPlay
  (bind (_x _y) (Math|Vector2D|BreakVector2D (Viewport|GetViewportSize)))
  (|CallParentFunction)
  (Effects|Components|ParticleSystem|SetFloatParameter (Variables|NiagaraActor|GetNiagaraComponent) "ViewportScale" (Math|Float|Max(Float) (* (/ (/ _x _y) 1.777778) 0.8) 1.0)))

(event EventTick
  (params (DeltaSeconds :type Float%20%28single-precision%29))
  (|CallParentFunction DeltaSeconds) ; This node is disabled and will not be called.\nDrag off pins to build functionality.
)

(fn ConstructionScript
  (|CallParentFunction))
