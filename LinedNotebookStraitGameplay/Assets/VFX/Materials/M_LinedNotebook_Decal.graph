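# Translucent deferred decal material for the LinedNotebook VFX, using the T_LinedNotebook_Decal texture's alpha as a mask.
# One chain multiplies that alpha by an OpacityMult parameter and decal lifetime opacity, so the mark follows the decal's fade.
# A separate multiply combines the same alpha with particle color alpha.
# Parameters expose a flat DecalColor, black by default, plus Metallic, Specular and Roughness.
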Material /LinedNotebookStraitGameplay/Assets/VFX/Materials/M_LinedNotebook_Decal
graph {
 properties {blendMode=BLEND_Translucent, materialDomain=MD_DeferredDecal, pixelDepthOffsetMode=PDOM_Legacy, usedShadingModels=Default Lit}
 nodes [
  n1 = TextureSample() {texture=/LinedNotebookStraitGameplay/Assets/VFX/Texture/T_LinedNotebook_Decal}
  Multiply(A=ParticleColor().A, B=@n1.A)
  Multiply(A=Multiply(A=@n1.A, B=ScalarParameter() {defaultValue=1, parameterName=OpacityMult}), B=DecalLifetimeOpacity().Opacity)
  ScalarParameter() {parameterName=Metallic}
  ScalarParameter() {parameterName=Specular}
  ScalarParameter() {defaultValue=1, parameterName=Roughness}
  VectorParameter() {parameterName=DecalColor, defaultValue=<rgba 0 0 0 1>}
 ]
}
